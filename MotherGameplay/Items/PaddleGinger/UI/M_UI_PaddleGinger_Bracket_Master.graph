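# Master translucent UI material for the PaddleGinger bracket graphic, with parameters for color, shape and glow.
# One texture packs the bracket shape in red, sharpened by subtract, hardness and saturate, and a glow in green with its own offset and strength.
# Their weighted sum forms a mask that blends from Falloff_color to Color and, scaled by FinalAlpha, yields an alpha value.
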Material /MotherGameplay/Items/PaddleGinger/UI/M_UI_PaddleGinger_Bracket_Master
graph {
 properties {blendMode=BLEND_Translucent, materialDomain=MD_UI, pixelDepthOffsetMode=PDOM_Legacy, usedShadingModels=Default Lit}
 nodes [
  n1 = TextureSample(UVs=TextureCoordinate()) {samplerType=SAMPLERTYPE_LinearColor, texture=&/T_UI_PaddleGinger_Bracket_128x}
  n2 = Multiply(
   A=Subtract(A=@n1.R, B=ScalarParameter() {defaultValue=0.5, parameterName=BracketSubtract})
   B=ScalarParameter() {defaultValue=4, parameterName=BracketHardness}
  )
  n3 = ScalarParameter() {parameterName=GlowSubtract}
  n4 = Add(
   A=Multiply(A=Saturate(None=@n2), B=ScalarParameter() {defaultValue=1, parameterName=BracketAmount})
   B=Multiply(A=Subtract(A=@n1.G, B=@n3), B=ScalarParameter() {defaultValue=1, parameterName=GlowAmount})
  )
  LinearInterpolate(
   A=VectorParameter() {parameterName=Falloff_color, defaultValue=<rgba 1 0 0 1>}.RGB
   B=VectorParameter() {parameterName=Color, defaultValue=<rgba 0 1 0 1>}.RGB
   Alpha=@n4
  )
  Multiply(A=@n4, B=ScalarParameter() {defaultValue=1, parameterName=FinalAlpha})
 ]
}
