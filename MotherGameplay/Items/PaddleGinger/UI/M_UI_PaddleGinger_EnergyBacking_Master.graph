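# Translucent UI master material that draws an animated, two-color energy backing for the PaddleGinger item.
# Two panning noise layers on square-corrected UVs, one distorting the other, drive time-offset sine waves that multiply into a shifting pattern.
# The pattern blends the two colors and builds opacity, with a static UsesFade switch adding a horizontal falloff shaped by a vertical sine.
# A radial-gradient bulge branch is present but has no visible consumer.
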Material /MotherGameplay/Items/PaddleGinger/UI/M_UI_PaddleGinger_EnergyBacking_Master
graph {
 properties {blendMode=BLEND_Translucent, materialDomain=MD_UI, pixelDepthOffsetMode=PDOM_Legacy, usedShadingModels=Default Lit}
 nodes [
  n1 = Divide(A=Time(), B=ScalarParameter() {defaultValue=5, group=Texture1, parameterName=BlocksTimeScale})
  n2 = Constant() {r=1}
  n3 = ScalarParameter() {defaultValue=1, group=Texture1, parameterName=Y_Scale_Correction}
  n4 = MaterialFunctionCall(
   UVs=MaterialFunctionCall() {materialFunction=/FortUILibrary/Foundation/Art_ToolKit/Tools/MF_UI_SquareToScreenSpace}.X_Priority
   Texture Scale=ScalarParameter() {defaultValue=1, parameterName=Scale}
  ) {materialFunction=&ScaleUVsByCenter}
  n5 = Multiply(A=Multiply(A=@n4.UVs, B=AppendVector(A=@n2, B=@n3)), B=ScalarParameter() {defaultValue=0.5, group=Texture1, parameterName=Tile})
  n6 = AppendVector(
   A=ScalarParameter() {defaultValue=0.06, group=Texture1, parameterName=PanSpeed_Noise_X}
   B=ScalarParameter() {defaultValue=0.05, group=Texture1, parameterName=PanSpeed_Noise_Y}
  )
  n7 = Multiply(A=@n4.UVs, B=AppendVector(A=Constant() {r=1}, B=ScalarParameter() {defaultValue=1, group=Texture2, parameterName=Y_Scale_Correction_2}))
  n8 = ScalarParameter() {defaultValue=3, group=Texture2, parameterName=Tile_Noise_2}
  n9 = ScalarParameter() {defaultValue=-0.125, group=Texture2, parameterName=PanSpeed_Noise_X_2}
  n10 = ScalarParameter() {defaultValue=-0.125, group=Texture2, parameterName=PanSpeed_Noise_Y_2}
  n11 = TextureSampleParameter2D(UVs=Panner(Coordinate=Multiply(A=@n7, B=@n8), Speed=AppendVector(A=@n9, B=@n10)) {speedX=-0.1}) {parameterName=Texture, samplerType=SAMPLERTYPE_LinearColor, texture=&/T_UI_PaddleGinger_Noise_512x}
  n12 = AppendVector(A=@n11.R, B=@n11.R)
  n13 = ScalarParameter() {defaultValue=0.25, group=Texture1, parameterName=AdditionalDistortion}
  n14 = TextureSampleParameter2D(UVs=Add(A=Panner(Coordinate=@n5, Speed=@n6) {speedX=-0.1}, B=Multiply(A=@n12, B=@n13))) {parameterName=Texture, samplerType=SAMPLERTYPE_LinearColor, texture=&T_UI_PaddleGinger_Noise_512x}
  n15 = ConstantBiasScale(None=Sine(None=Add(A=@n1, B=@n14.R)) {period=0.15})
  n16 = Add(A=Divide(A=Time(), B=ScalarParameter() {defaultValue=16, group=Texture2, parameterName=BlocksTimeScale_2}), B=@n11.B)
  n17 = OneMinus(None=ConstantBiasScale(None=Sine(None=@n16) {period=0.15}))
  n18 = Multiply(A=Saturate(None=DotProduct(A=@n15, B=@n17)), B=ScalarParameter() {defaultValue=0.5, parameterName=PatternAmount})
  n19 = TextureCoordinate()
  n20 = OneMinus(None=ComponentMask(None=@n19) {r=true})
  n21 = ScalarParameter() {defaultValue=0.5, parameterName=BackingAmount}
  n22 = ComponentMask(None=Sine(None=@n19) {period=2}) {g=true}
  n23 = ScalarParameter() {defaultValue=1, parameterName=VeritcalPower}
  n24 = Saturate(None=DotProduct(A=Multiply(A=@n20, B=@n21), B=Power(Base=@n22, Exp=@n23)))
  n25 = DotProduct(A=@n18, B=@n24)
  n26 = Saturate(None=Add(A=Saturate(None=@n25), B=@n24))
  LinearInterpolate(
   A=VectorParameter() {group=Color, parameterName=Color2, defaultValue=<rgba 0 0.5 1 1>}.RGB
   B=VectorParameter() {group=Color, parameterName=Color1, defaultValue=<rgba 0.25 1 0.5 1>}.RGB
   Alpha=StaticSwitchParameter(True=@n26, False=@n18) {defaultValue=true, parameterName=UsesFade}
  )
  n27 = MaterialFunctionCall() {materialFunction=&RadialGradientExponential}
  n28 = OneMinus(None=@n27.RadialGradientExponential)
  n29 = ScalarParameter() {group=Distort, parameterName=Flip_Bulge_Direction}
  Multiply(A=LinearInterpolate(
    A=ScalarParameter() {defaultValue=0.5, group=Distort, parameterName=Bulge_Scale_A}
    B=ScalarParameter() {defaultValue=1, group=Distort, parameterName=Bulge_Scale_B}
    Alpha=LinearInterpolate(A=@n27.RadialGradientExponential, B=@n28, Alpha=@n29)
   ))
  n30 = ScalarParameter() {defaultValue=0.5, parameterName=BackingAmount}
  StaticSwitchParameter(True=@n26, False=Saturate(None=Add(A=@n30, B=@n18))) {defaultValue=true, parameterName=UsesFade}
  TextureCoordinate()
 ]
}
