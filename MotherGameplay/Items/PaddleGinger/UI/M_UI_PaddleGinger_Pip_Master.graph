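# Translucent UI master material that draws a round pip for the PaddleGinger item's UI, with selected and unselected looks.
# An exponential radial gradient forms the dot, with a parameter controlling edge hardness.
# The IsSelected scalar blends the gradient radius between start and stop scales, the color between Color_1 and SelectedColor, and the alpha multiplier between NotSelectedAlpha and FinalAlpha.
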Material /MotherGameplay/Items/PaddleGinger/UI/M_UI_PaddleGinger_Pip_Master
graph {
 properties {blendMode=BLEND_Translucent, materialDomain=MD_UI, pixelDepthOffsetMode=PDOM_Legacy, usedShadingModels=Default Lit}
 nodes [
  LinearInterpolate(
   A=VectorParameter() {parameterName=Color_1, defaultValue=<rgba 1 0 0 1>}.RGB
   B=VectorParameter() {parameterName=SelectedColor, defaultValue=<rgba 0 1 0 1>}.RGB
   Alpha=ScalarParameter() {defaultValue=1, parameterName=IsSelected, sliderMax=1}
  )
  n1 = ScalarParameter() {defaultValue=0.4, parameterName=ScaleStart}
  n2 = ScalarParameter() {defaultValue=0.5, parameterName=ScaleStop}
  n3 = ScalarParameter() {defaultValue=1, parameterName=IsSelected, sliderMax=1}
  Multiply(
   A=MaterialFunctionCall(Radius=LinearInterpolate(A=@n1, B=@n2, Alpha=@n3), Density=ScalarParameter() {defaultValue=24, parameterName=PipEdgeHardness}) {materialFunction=&RadialGradientExponential}.RadialGradientExponential
   B=LinearInterpolate(
    A=ScalarParameter() {defaultValue=0.5, parameterName=NotSelectedAlpha}
    B=ScalarParameter() {defaultValue=1, parameterName=FinalAlpha}
    Alpha=ScalarParameter() {defaultValue=1, parameterName=IsSelected, sliderMax=1}
   )
  )
 ]
}
