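# Translucent UI master material for the PaddleGinger item that draws a thin outline with adjustable rounded corners.
# An outer rounded-box shape, built procedurally from centered 9-slice UVs, is multiplied by an inverted inner one, each with its own size, inset and hardness.
# A rotatable, offset linear mask trims the line.
# The result, scaled by Final_Alpha, blends FalloffColor into Color.
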Material /MotherGameplay/Items/PaddleGinger/UI/M_UI_PaddleGinger_ThinLine_Master
graph {
 properties {blendMode=BLEND_Translucent, materialDomain=MD_UI, pixelDepthOffsetMode=PDOM_Legacy, usedShadingModels=Default Lit}
 nodes [
  n1 = ConstantBiasScale(None=MaterialFunctionCall() {materialFunction=&GetUserInterfaceUV}."9-Slice UV (No Tiling)") {bias=-0.5, scale=2}
  n2 = Abs(None=Divide(A=@n1, B=ScalarParameter() {defaultValue=1, parameterName=Size}))
  n3 = OneMinus(None=AppendVector(A=ScalarParameter() {defaultValue=0.25, parameterName=Curve_X}, B=ScalarParameter() {defaultValue=0.25, parameterName=Curve_Y}))
  n4 = Subtract(B=@n3)
  n5 = Saturate(None=Divide(A=Subtract(A=@n2, B=@n3) {constB=0.25}, B=@n4))
  n6 = DotProduct(A=@n5, B=@n5)
  n7 = Subtract(A=OneMinus(None=Saturate(None=@n6)), B=ScalarParameter() {defaultValue=0.005, parameterName=Subtract})
  n8 = ScalarParameter() {defaultValue=4, parameterName=Hardness}
  n9 = Abs(None=Divide(A=@n1, B=ScalarParameter() {defaultValue=1, parameterName=SizeInner}))
  n10 = Saturate(None=Divide(A=Subtract(A=@n9, B=@n3) {constB=0.25}, B=@n4))
  n11 = OneMinus(None=Saturate(None=DotProduct(A=@n10, B=@n10)))
  n12 = ScalarParameter() {defaultValue=0.005, parameterName=SubtractInner}
  n13 = Multiply(A=Saturate(None=Subtract(A=@n11, B=@n12)), B=ScalarParameter() {defaultValue=4, parameterName=HardnessInner})
  n14 = DotProduct(A=Saturate(None=Multiply(A=@n7, B=@n8)), B=OneMinus(None=Saturate(None=@n13)))
  n15 = ScalarParameter() {group=Mask, parameterName=Mask_Rotation}
  n16 = Add(
   A=MaterialFunctionCall() {materialFunction=&GetUserInterfaceUV}."9-Slice UV (No Tiling)"
   B=ScalarParameter() {defaultValue=-0.325, group=Mask, parameterName=Mask_Offset}
  ) {constB=-0.3275}
  n17 = Multiply(
   A=ComponentMask(None=MaterialFunctionCall!(Rotation_Fraction=@n15, TexCoord=@n16) {materialFunction=/FortUILibrary/Foundation/Art_ToolKit/AngledBoxes/MF_UI_Rotator_Fractions}.Result) {r=true}
   B=ScalarParameter() {defaultValue=16, group=Mask, parameterName=Mask_Egde}
  )
  n18 = Saturate(None=DotProduct(A=Saturate(None=@n14), B=Saturate(None=@n17)))
  n19 = ScalarParameter() {defaultValue=1, parameterName=Final_Alpha}
  LinearInterpolate(
   A=VectorParameter() {parameterName=FalloffColor, defaultValue=<rgba 0 1 0 1>}.RGB
   B=VectorParameter() {parameterName=Color, defaultValue=<rgba 1 0 0 1>}.RGB
   Alpha=Saturate(None=Multiply(A=@n18, B=@n19))
  )
 ]
}
